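# Physics asset for the medium male high-motion backpack cosmetic, giving the pack its collision and physics bodies.
# Two thin, overlapping box bodies, slightly offset in depth, cover the dyn_pack_base and pack bones.
# A single constraint joins them with both swings and twist locked, so the configured angle limits are unused and neither body rotates relative to the other.
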PhysicsAsset /BRCosmetics/Accessories/FORT_Backpacks/Backpack_M_MED_HighMotion/Meshes/M_MED_HighMotion_Pack_Physics
bodies {
 dyn_pack_base [{shapeType=Box, extentX=22.5, extentY=5.5, extentZ=27, shapeName=dyn_pack_base_box, center=<0 -0.209 0>}]
 pack [{shapeType=Box, extentX=23.5, extentY=2, extentZ=26.5, shapeName=pack_box, center=<0 1.6475 0>}]
}
constraints [{
  bone1Name=dyn_pack_base
  bone2Name=pack
  swing1LimitDegrees=45
  swing1Motion=Locked
  swing2LimitDegrees=15
  swing2Motion=Locked
  twistLimitDegrees=45
  twistMotion=Locked
 }]
